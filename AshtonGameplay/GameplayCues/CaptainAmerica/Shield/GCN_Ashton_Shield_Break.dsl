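; Gameplay cue notify in the Ashton Captain America shield folder that plays shield-break hit-notify audio when it bursts.
; OnBurst casts the cue's source object to a FortPawn and gets that pawn's HitNotify SoundLibrary component.
; If valid, it configures the context from the effect causer and pawn, then plays the BShieldBreak event.
; Routing through SoundLibrary lets Experiences decide whether the sound plays.
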(fn OnBurst(EitherVFX-preferthisversion)
  (params (MyTarget :type Actor%20Object%20Reference) (Parameters :type Gameplay%20Cue%20Parameters%20Structure%20%28by%20ref%29) (ParticleComponents :type Array%20of%20FXSystem%20Component%20Object%20References) (AudioComponents :type Array%20of%20Audio%20Component%20Object%20References) (BurstCameraShakeInstance :type Camera%20Shake%20Base%20Object%20Reference) (BurstDecalInstance :type Decal%20Actor%20Object%20Reference))
  (bind (_normalizedmagnitude _rawmagnitude _effectcontext _matchedtagname _originaltag _aggregatedsourcetags _aggregatedtargettags _location _normal _instigator _effectcauser _sourceobject) (Ability|GameplayCue|BreakGameplayCueParameters Parameters))
  ;[0 0 0] New way of doing things with SoundLibrary, allows Experiences to specify if they want sounds to play
  (bind _asfort_pawn (Utilities|Casting|CastToFortPawn _sourceobject)
    (:then
      (bind _asfort_hit_notify_sound_library_component (Utilities|Casting|CastToFortHitNotifySoundLibraryComponent (SoundLibrary|GetSoundLibraryComponent _asfort_pawn "(TagName=\"SoundLibrary.HitNotify\")" :!class /Script/FortniteGame.FortPawn)))
      (Utilities|IsValid (SoundLibrary|GetSoundLibraryComponent _asfort_pawn "(TagName=\"SoundLibrary.HitNotify\")" :!class /Script/FortniteGame.FortPawn)
        (:Is%20Valid
          (SoundLibrary|ConfigureContext _asfort_hit_notify_sound_library_component (Utilities|Struct|MakeFortHitNotifyContextSettings _effectcauser _asfort_pawn) :!class /Script/FortniteGame.FortHitNotifySoundLibraryComponent)
          (SoundLibrary|PlayEvent _asfort_hit_notify_sound_library_component "(TagName=\"SoundLibrary.HitNotify.Event.BShieldBreak\")"))))
    (:CastFailed
      (return))))
