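; Niagara burst gameplay cue for Captain America's shield hitting the environment.
; After the parent burst, it plays an impact sound at the cue location, passing the truncated normalized magnitude as ImpactCount.
; It flags the particle effect as a non-player hit and feeds it the target's forward vector as WeaponVec.
; Where Janus is allowed and the physics field console variable is on, it spawns a radial force field there.
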(fn OnBurst(Niagara)
  (params (MyTarget :type Actor%20Object%20Reference) (Parameters :type Gameplay%20Cue%20Parameters%20Structure%20%28by%20ref%29) (ParticleComponents :type Array%20of%20Niagara%20Particle%20System%20Component%20Object%20References) (AudioComponents :type Array%20of%20Audio%20Component%20Object%20References) (BurstCameraShakeInstance :type Camera%20Shake%20Base%20Object%20Reference) (BurstDecalInstance :type Decal%20Actor%20Object%20Reference))
  (bind (_normalizedmagnitude _rawmagnitude _effectcontext _matchedtagname _originaltag _aggregatedsourcetags _aggregatedtargettags _location) (Ability|GameplayCue|BreakGameplayCueParameters Parameters))
  (bind _output (Utilities|Array|Get(acopy) ParticleComponents))
  (|CallParentFunction MyTarget Parameters ParticleComponents AudioComponents BurstCameraShakeInstance BurstDecalInstance)
  (bind _returnvalue (Audio|SpawnSoundatLocation 0 (Variables|Default|GetImpactSound) _location))
  (Audio|Parameter|SetIntegerParameter _returnvalue "ImpactCount" (Math|Float|Truncate _normalizedmagnitude) :!class /Script/Engine.AudioComponent)
  (Effects|Components|ParticleSystem|SetBoolParameter _output "HitPlayer" false)
  (Effects|Components|ParticleSystem|SetVectorParameter _output "WeaponVec" (Transformation|GetActorForwardVector MyTarget))
  ;[1 1 1] Add RadialForce
  (Utilities|AllowJanus
    (:Allow
      (bind _returnvalue_2 (Development|GetConsoleVariableIntValue "r.PhysicsField.EnableField"))
      (if (== _returnvalue_2 1)
        (Game|SpawnActorfromClass 0 "/Game/Effects/Niagara/Blueprints/PhsyicsFields/FN_RadialForce.FN_RadialForce_C" (Math|Transform|MakeTransform _location) "Undefined" "MultiplyWithRoot" 0 0.2 0.75 4e+02 0.5 0.0 0)))
    (:Nope
      _))
  (return))
